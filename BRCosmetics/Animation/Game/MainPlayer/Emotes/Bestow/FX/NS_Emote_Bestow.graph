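# Niagara system for the Bestow emote, staging an award-style moment with confetti, spotlights and camera flashes.
# Confetti squares and mesh streamers burst downward from two raised side points, colored from a user color array.
# Warm flickering spotlights, floor glows and slowly rotating ground light meshes last about eleven seconds, while photo flashes layer sprites, cones and lights on a ramping spawn rate.
# Spawn groups vary per-particle placement; time-of-day and front-end modifiers adjust brightness and scale.
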NiagaraSystem /BRCosmetics/Animation/Game/MainPlayer/Emotes/Bestow/FX/NS_Emote_Bestow
dependencies {
 usedDataInterfaces [
  &NiagaraDataInterfaceCurve
  &NiagaraDataInterfaceColorCurve
  &NiagaraDataInterfaceArrayFloat3
  &NiagaraDataInterfaceSpriteRendererInfo
  &NiagaraDataInterfaceArrayInt32
  &NiagaraDataInterfaceArrayFloat
 ]
 usedDynamicInputs []
 usedModules [&SystemState]
}
emitters {
 ConfettiSquares {
  emitterUpdateScript [
   &EmitterState()
   &SpawnBurst_Instantaneous(Spawn Count=30, Use Spawn Probability=false, UseLoopCountLimit=false)
  ]
  particleSpawnScript [
   &FN_SetSpawnGroup(Number of Spawn Groups=2)
   &InitializeParticle(
    Color=&SelectLinearColorFromArray(Array Sampling Mode=::NewEnumerator0, Direct Array Index=?, Color Selection Array=$User.ConfettiColorArray, Random Seed=0)
    Lifetime Max=3.333
    Lifetime Mode=::NewEnumerator1
    Position Offset=&Add_Vector(
     A=&SelectVectorFromArray(
      Direct Array Index=&SpawnGroup()
      Vector Selection Array=DataInterface() {propertyValues {FloatData [<70 0 250>, <-70 0 250>]}}
     )
     B=&RandomRangeVector(Minimum=<0 0 20>, Maximum=<0 0 50>)
    )
    Sprite Size Mode=::NewEnumerator1
    Uniform Sprite Size Max=6
    Uniform Sprite Size Min=3
    UsePositionOffset=true
    Write Material Random=true
   )
   &AddVelocity(
    Cone Angle=60
    Cone Axis=&SelectVectorFromArray(
     Direct Array Index=&SpawnGroup()
     Vector Selection Array=DataInterface() {propertyValues {FloatData [<-0.4 0 -1>, <0.4 0 -1>]}}
    )
    Module.UseConeFalloff=true
    Module.UseVelDistribution=true
    Override Seed=false
    Velocity Mode=::NewEnumerator2
    Velocity Speed=&RandomRangeFloat(Minimum=300, Maximum=1111)
   )
  ]
  particleUpdateScript [
   &ParticleState()
   &GravityForce(Gravity=<0 0 -111>)
   /Niagara/Modules/Update/Forces/CurlNoiseForce(Noise Frequency=0.02, Noise Strength=200, Pan Noise=false)
   &Drag(
    Drag=&FloatFromCurve(FloatCurve=DataInterface() {propertyValues {Curve {keys [<at 0 0.5>,<at 1 1>]}}}, Scale Curve=6)
    Use Linear Drag=true
    Use Rotational Drag=true
   )
   &SolveForcesAndVelocity(Clamp Velocity=false, Limit Acceleration=false)
   &ScaleSpriteSize(Scale Factor=&Vector2DFromFloat(Value=&FloatFromCurve(FloatCurve=DataInterface() {propertyValues {Curve {keys [
          <at 0 0 RCIM_Cubic>
          {arriveTangent=-0.0623, interpMode=RCIM_Cubic, leaveTangent=-0.0623, tangentMode=RCTM_User, time=0.04, value=1}
          {arriveTangent=-0.0083, interpMode=RCIM_Cubic, leaveTangent=-0.0083, tangentMode=RCTM_User, time=0.5, value=1}
          <at 1 0 RCIM_Cubic>
         ]}}})))
   &SubUVAnimation(
    "Sprite Renderer "=DataInterface() {propertyValues {SpriteRenderer=&self:ConfettiSquares_0.NiagaraSpriteRendererProperties_0}}
    Start Frame Offset:NiagaraInt32=&RandomRangeInt(Minimum=0, Maximum=555)
    UseEndFrame=false
    UseStartFrame=false
   )
   &ScaleColor(
    Scale RGB=&VectorFromFloat(Value=&FNTimeOfDayScalar(Night Scale=0.25, Morning Scale=0.333, Day Scale=0.5, Evening Scale=0.25))
    ScaleA=true
    ScaleRGB=true
   )
  ]
  rendererClasses [&NiagaraSpriteRendererProperties]
  renderers [{rendererClass=&NiagaraSpriteRendererProperties}]
 }
 ConfettiStreamers {
  emitterUpdateScript [
   &EmitterState()
   &SpawnBurst_Instantaneous(Spawn Count=8, Use Spawn Probability=false, UseLoopCountLimit=false)
  ]
  particleSpawnScript [
   &FN_SetSpawnGroup(Number of Spawn Groups=2)
   &InitializeParticle(
    Color=&SelectLinearColorFromArray(Array Sampling Mode=::NewEnumerator0, Direct Array Index=?, Color Selection Array=$User.ConfettiColorArray, Random Seed=0)
    Lifetime=&RandomRangeFloat(Minimum=1, Maximum=1.7)
    Mesh Scale Mode=::NewEnumerator3
    Mesh Uniform Scale=&RandomRangeFloat(Minimum=0.3, Maximum=0.6)
    Position Offset=&Add_Vector(
     A=&SelectVectorFromArray(
      Direct Array Index=&SpawnGroup()
      Vector Selection Array=DataInterface() {propertyValues {FloatData [<70 0 200>, <-70 0 200>]}}
     )
     B=&RandomRangeVector(Minimum=<0 0 60>, Maximum=<0 0 100>)
    )
    UsePositionOffset=true
    Write Material Random=true
   )
   &AddVelocity(
    Cone Angle=50
    Cone Axis=&SelectVectorFromArray(
     Direct Array Index=&SpawnGroup()
     Vector Selection Array=DataInterface() {propertyValues {FloatData [<0 0 -1>, <0 0 -1>]}}
    )
    Module.UseConeFalloff=true
    Module.UseVelDistribution=true
    Override Seed=false
    Velocity Mode=::NewEnumerator2
    Velocity Speed=&RandomRangeFloat(Minimum=400, Maximum=900)
   )
   &InitialMeshOrientation(
    Mesh Orientation Mode=::NewEnumerator2
    Orientation Coordinate Space=::Local
    Orientation Vector=$Particles.Velocity
    Rotation=&SelectVectorFromArray(
     Direct Array Index=&SpawnGroup()
     Vector Selection Array=DataInterface() {propertyValues {FloatData [<0.125 0 0>, <0.44 0 0>]}}
    )
    Use Rotation Vector=true
   )
  ]
  particleUpdateScript [
   &ParticleState()
   &GravityForce(Gravity=<0 0 -111>)
   &Drag(Drag=6.5, Use Linear Drag=true, Use Rotational Drag=false)
   &SolveForcesAndVelocity(Clamp Velocity=false, Limit Acceleration=false)
   &ScaleColor(
    Scale RGB=&VectorFromFloat(Value=&FNTimeOfDayScalar(Night Scale=0.3, Morning Scale=0.333, Day Scale=0.666, Evening Scale=0.333))
    ScaleA=false
    ScaleRGB=true
   )
  ]
  rendererClasses [&NiagaraMeshRendererProperties]
  renderers [{rendererClass=&NiagaraMeshRendererProperties, rendererIndex=0}]
 }
 GroundLightMesh {
  emitterUpdateScript [
   &EmitterState(Inactive Response=::NewEnumerator1, Life Cycle Mode=::NewEnumerator1, Loop Behavior=::NewEnumerator1, Loop Duration=13, UseLoopDelay=false)
   &SpawnBurst_Instantaneous(Spawn Count=7, Use Spawn Probability=false, UseLoopCountLimit=false)
  ]
  particleSpawnScript [
   &FN_SetSpawnGroup(Number of Spawn Groups=7)
   &InitializeParticle(
    Lifetime=11
    Mesh Scale=&SelectVectorFromArray(
     Direct Array Index=$Particles.SpawnGroup
     Vector Selection Array=DataInterface() {propertyValues {FloatData [<0.3 0.3 0.5>, <0.3 0.3 1.6>, <0.3 0.3 1.6>, <0.25 0.25 0.5>, <0.3 0.3 0.5>, <0.3 0.3 1.6>, <0.4 0.4 0.4>]}}
    )
    Mesh Scale Mode=::NewEnumerator4
    Position Offset=&Multiply_Vector(
     A=&SelectVectorFromArray(
      Direct Array Index=$Particles.SpawnGroup
      Vector Selection Array=DataInterface() {propertyValues {FloatData [<30 20 5>, <30 -20 5>, <0 35 5>, <0 -35 5>, <-40 20 5>, <-40 -20 5>, <0 0 3>]}}
     )
     B=<1.6 1.6 1>
    )
    UsePositionOffset=true
    Write Material Random=true
   )
   &InitialMeshOrientation(
    Mesh Orientation Mode=::NewEnumerator7
    Rotation=&SelectVectorFromArray(
     Direct Array Index=$Particles.SpawnGroup
     Vector Selection Array=DataInterface() {propertyValues {FloatData [<0.25 0.5 0>, <0.1892 0.4009 0>, <0.5653 0.0405 0>, <0.25 0.5 0>, <0.25 0.5 0>, <-0.2275 0.527 0>, <0 0 0>]}}
    )
    Use Rotation Vector=true
   )
   &DynamicMaterialParameters(
    Index 0 Param 1=10
    Index 0 Param 2=1
    Index 0 Param 3=1
    Index 0 Param 4=2
    Index 1 Param 1=15
    Module.Param0WriteEnabled=true
    Module.Param0WriteEnabled1=true
    Module.Param1WriteEnabled=true
    Module.Param1WriteEnabled1=true
    Module.Param2WriteEnabled=true
    Module.Param2WriteEnabled1=true
    Module.Param3WriteEnabled=true
    Module.Param3WriteEnabled1=true
    Write Parameter Index 1=true
   )
   /SetVariables(Particles.GlowGroundVisibility:NiagaraInt32=&SelectIntFromArray(
     Direct Array Index=&SpawnGroup()
     Int Selection Array=DataInterface() {propertyValues {IntData [0,0,0,0,0,0,1]}}
    )) {isSetParameters=true}
  ]
  particleUpdateScript [
   &ParticleState()
   &UpdateMeshOrientation(
    Rotation Coordinate Space=::NewEnumerator0
    Rotation Rate=&FloatFromCurve(
     FloatCurve=DataInterface() {propertyValues {Curve {keys [
         <at 0.1112 0 RCIM_Cubic>
         {arriveTangent=3.3189, interpMode=RCIM_Cubic, leaveTangent=3.3189, time=0.2104, value=1.5}
         <at 0.4125 1 RCIM_Cubic>
        ]}}}
     Scale Curve=&SelectFloatFromArray(
      Direct Array Index=$Particles.SpawnGroup
      Float Selection Array=DataInterface() {propertyValues {FloatData [0.3,-0.35,0.25,-0.38,0.4,-0.3,0]}}
     )
    )
    Rotation Vector=<0 0.14 1>
   )
   &Color(Color=&ColorFromCurve(ColorCurve=DataInterface() {propertyValues {
       AlphaCurve {keys [<at 0 0.75 RCIM_Cubic>]}
       BlueCurve {keys [<at 0.1488 0.3945 RCIM_Cubic>,<at 0.3893 0.3245>,<at 0.534 0.375>,<at 0.7806 0.3202>]}
       GreenCurve {keys [<at 0.1488 0.6366 RCIM_Cubic>,<at 0.3893 0.5478>,<at 0.534 0.7633>,<at 0.7806 0.5112>]}
       RedCurve {keys [<at 0.1488 0.75 RCIM_Cubic>,<at 0.3893 1>,<at 0.534 1>,<at 0.7806 0.8958>]}
      }})) {module=Color001}
   &ScaleColor(
    Color Value To Scale=$Particles.Color
    Scale Alpha=&Multiply_Float(
     A=&FloatFromCurve(
      FloatCurve=DataInterface() {propertyValues {Curve {keys [
          {arriveTangent=-2.5, interpMode=RCIM_Cubic, time=0, value=0}
          {arriveTangent=3.4501, interpMode=RCIM_Cubic, leaveTangent=3.4501, time=0.0814, value=1}
          {arriveTangent=-0.3453, interpMode=RCIM_Cubic, leaveTangent=-0.3453, time=0.2, value=0.69}
          {arriveTangent=0.22, interpMode=RCIM_Cubic, leaveTangent=0.22, time=0.4, value=0.89}
          {arriveTangent=0.2369, interpMode=RCIM_Cubic, leaveTangent=0.2369, time=0.7, value=0.8}
          {arriveTangent=-2.6667, interpMode=RCIM_Cubic, leaveTangent=-2.6667, time=0.8643, value=1}
          <at 1 0 RCIM_Cubic>
         ]}}}
      Scale Curve=0.6
     )
     B=1
    )
    ScaleA=true
    ScaleRGB=true
   )
   &ScaleMeshSize(Scale Factor=&VectorFromFloat(Value=0.6))
   &InheritSourceMovement(ApplyForce=false, ApplyPosition=true, ApplyPositionFalloff=false)
   &SolveForcesAndVelocity(Clamp Velocity=false, Limit Acceleration=false)
   /SetVariables(Particles.GlowGroundColor:LinearColor=&ColorFromCurve(ColorCurve=DataInterface() {propertyValues {
       AlphaCurve {keys [
         <at 0 0>
         {arriveTangent=0.8771, interpMode=RCIM_Cubic, leaveTangent=0.8771, time=0.0798, value=0.27}
         <at 0.2736 0.24>
         <at 0.4888 0.27>
         <at 0.7204 0.24>
         <at 0.9048 0.27>
         <at 1 0>
        ]}
       BlueCurve {keys [<at 0.1056 1.9271 RCIM_Cubic>,<at 0.2928 2.2455>,<at 0.461 1.8924>,<at 0.6568 2.1506>,<at 0.8577 2.6823>]}
       GreenCurve {keys [<at 0.1056 3.2075 RCIM_Cubic>,<at 0.2928 2.9502>,<at 0.461 2.5953>,<at 0.6568 3.0104>,<at 0.8577 3.648>]}
       RedCurve {keys [<at 0.1056 5 RCIM_Cubic>,<at 0.2928 5>,<at 0.461 5>,<at 0.6568 5>,<at 0.8577 5>]}
      }})) {isSetParameters=true}
  ]
  rendererClasses [&NiagaraMeshRendererProperties, &NiagaraSpriteRendererProperties]
  renderers [
   {rendererClass=&NiagaraMeshRendererProperties, rendererIndex=0}
   {rendererClass=&NiagaraSpriteRendererProperties, rendererIndex=1}
   {rendererClass=&NiagaraMeshRendererProperties, rendererIndex=2}
  ]
 }
 PhotoFlashes {
  emitterUpdateScript [
   &EmitterState(Inactive Response=::NewEnumerator1, Life Cycle Mode=::NewEnumerator1, Loop Behavior=::NewEnumerator1, Loop Duration=11, UseLoopDelay=false)
   &SpawnRate(
    SpawnRate=&FloatFromCurve(
     FloatCurve=DataInterface() {propertyValues {Curve {keys [
         {arriveTangent=4.8497, interpMode=RCIM_Cubic, leaveTangent=4.8497, tangentMode=RCTM_User, time=0, value=0}
         <at 0.6 1 RCIM_Cubic>
         {arriveTangent=-5.3133, tangentMode=RCTM_User, time=0.8841, value=0}
        ]}}}
     CurveIndex=$Emitter.NormalizedLoopAge
     Scale Curve=5
    )
    Use Spawn Probability=false
   )
  ]
  particleSpawnScript [
   &InitializeParticle(
    Lifetime Max=0.17
    Lifetime Min=0.13
    Lifetime Mode=::NewEnumerator1
    Mass Max=2
    Mass Mode=::NewEnumerator3
    Position Offset=<0 80 130>
    Position Offset Coordinate Space=::Local
    Sprite Rotation Mode=::NewEnumerator2
    Sprite Size Mode=::NewEnumerator3
    Uniform Sprite Size=&RandomRangeFloat(Minimum=10, Maximum=25)
    UsePositionOffset=true
    Write Material Random=true
   )
   &CameraOffset(Camera Offset Amount=20) {module=CameraOffset001}
   &ShapeLocation(
    Cylinder Height=80
    Cylinder Radius=&FN_FrontEndFloatModifier(Gameplay=120, Front End=80)
    Hemicircle Y=true
    MaskSpawnGroup=false
    Override Seed=false
    Rotation Angle=1
    Rotation Axis=<0 0 1>
    Rotation Mode=::NewEnumerator1
    Shape Primitive=::NewEnumerator1
    Surface Only=false
    UseOwnerScale=true
   ) {module=ShapeLocation001}
   /SetVariables(Particles.LightVisability:NiagaraInt32=&RandomRangeInt(Minimum=0, Maximum=1)) {isSetParameters=true}
   /SetVariables(Particles.Gen5LightDirection:Vector3f=<-15 -90 0>) {isSetParameters=true}
   &InitialMeshOrientation(
    Mesh Axis To Orient=<0 0 -1>
    Mesh Orientation Mode=::NewEnumerator2
    Orientation Vector=&NormalizeVector(
     Vector To Normalize=&Subtract_Vector(A=/Niagara/DynamicInputs/Helpers/SimulationPosition(), B=Linked() {linkedVariable {name=Particles.Position, type=NiagaraPosition}})
     Safe Normalize=true
     Safe Fallback Threshold=0
     Safe Fallback Vector=<0 0 0.001>
    )
    Rotation=<-0.126 0 0>
    Use Rotation Vector=true
   )
  ]
  particleUpdateScript [
   &ParticleState()
   &ScaleColor(
    Scale Alpha=&FloatFromCurve(
     FloatCurve=DataInterface() {propertyValues {Curve {keys [
         <at 0 0 RCIM_Cubic>
         {arriveTangent=0.1895, interpMode=RCIM_Cubic, leaveTangent=0.1895, tangentMode=RCTM_User, time=0.1, value=1}
         {arriveTangent=-0.0202, interpMode=RCIM_Cubic, leaveTangent=-0.0202, tangentMode=RCTM_User, time=0.35, value=1}
         <at 1.0011 0.002 RCIM_Cubic>
        ]}}}
     Scale Curve=&FN_FrontEndFloatModifier(Gameplay=9, Front End=6)
    )
    ScaleA=true
    ScaleRGB=true
   )
   &ScaleSpriteSize(
    Uniform Curve Scale=&FN_FrontEndFloatModifier(
     Gameplay=&RandomRangeFloat(Minimum=1.2, Maximum=2)
     Front End=&RandomRangeFloat(Minimum=0.6, Maximum=0.9)
    )
    Uniform Curve Sprite Scale=DataInterface() {propertyValues {Curve {keys [
        {arriveTangent=3.1344, interpMode=RCIM_Cubic, leaveTangent=3.1344, tangentMode=RCTM_User, time=0, value=0}
        {tangentMode=RCTM_User, time=1, value=1}
       ]}}}
   )
   /SetVariables(
    Particles.StarColor:LinearColor=&MultiplyLinearColor(Linear Color=$Particles.Color, Scale Factor=&MakeVector4_FromLinearColor(LinearColor=<rgba 1 1 1 1>))
    Particles.StarRotation:NiagaraFloat=&RandomRangeFloat(Minimum=0, Maximum=360)
    Particles.StarSize:Vector2f=&Multiply_Vector2(A=$Particles.SpriteSize, B=<2 2>)
   ) {isSetParameters=true}
   /SetVariables(
    Particles.StreaksColor:LinearColor=&MultiplyLinearColor(
     Linear Color=$Particles.Color
     Scale Factor=&MakeVector4_FromLinearColor(LinearColor=<rgba 0.6 0.8 1 0.01>)
    )
    Particles.StreaksRotation:NiagaraFloat=&RandomRangeFloat(Minimum=0, Maximum=360)
    Particles.StreaksSize:Vector2f=&Multiply_Vector2(A=$Particles.SpriteSize, B=<7 7>)
   ) {isSetParameters=true}
   /SetVariables(
    Particles.AnamorphicColor:LinearColor=&MultiplyLinearColor(
     Linear Color=$Particles.Color
     Scale Factor=&MakeVector4_FromLinearColor(LinearColor=<rgba 0.8 0.9 1 0.01>)
    )
    Particles.AnamorphicRotation=90
    Particles.AnamorphicSize:Vector2f=&Multiply_Vector2(A=$Particles.SpriteSize, B=<3 9>)
   ) {isSetParameters=true}
   /SetVariables(
    Particles.HaloColor:LinearColor=&MultiplyLinearColor(Linear Color=$Particles.Color, Scale Factor=&MakeVector4_FromLinearColor(LinearColor=<rgba 1 1 1 0.025>))
    Particles.HaloSize:Vector2f=&Multiply_Vector2(A=$Particles.SpriteSize, B=<5 5>)
   ) {isSetParameters=true}
   /SetVariables(
    Particles.HaloColor_02:LinearColor=&MultiplyLinearColor(Linear Color=$Particles.Color, Scale Factor=&MakeVector4_FromLinearColor(LinearColor=<rgba 1 1 1 0.05>))
    Particles.HaloSize_02:Vector2f=&Multiply_Vector2(A=$Particles.SpriteSize, B=<1.5 1.5>)
   ) {isSetParameters=true}
   /SetVariables(
    Particles.AirGlowColor:LinearColor=&MultiplyLinearColor(Linear Color=$Particles.Color, Scale Factor=&MakeVector4_FromLinearColor(LinearColor=<rgba 1 1 1 0.005>))
    Particles.AirGlowSize:Vector2f=&Multiply_Vector2(A=$Particles.SpriteSize, B=<4 4>)
   ) {isSetParameters=true}
   &FN_Light_Attributes(
    Color=<rgba 3 3 3 0.2>
    Create Color Attribute=true
    Create Exponent Attribute=true
    Create Position Attribute=true
    Create Radius Attribute=true
    Enable Particle Light Enable=true
    Exponent=18
    Light Enabled=true
    Position=&AddVectorToPosition(Position=$Particles.Position, Vector=<0 -50 0>)
    Radius=140
    Scale Color=&FloatFromCurve(
     FloatCurve=DataInterface() {propertyValues {Curve {keys [
         {arriveTangent=9.0725, interpMode=RCIM_Cubic, leaveTangent=9.0725, tangentMode=RCTM_User, time=0, value=0}
         {arriveTangent=0.0246, interpMode=RCIM_Cubic, leaveTangent=0.0246, tangentMode=RCTM_User, time=0.1408, value=0.9847}
         {arriveTangent=-0.7974, interpMode=RCIM_Cubic, leaveTangent=-0.7974, tangentMode=RCTM_User, time=1, value=0}
        ]}}}
     Scale Curve=&FN_FrontEndFloatModifier(Gameplay=10, Front End=7)
    )
   )
   /SetVariables(Particles.TempPosition:Vector3f=$Particles.FN_Light_Attributes.Position) {isSetParameters=true}
   &FN_Gen5_Light(
    Attenutation:NiagaraFloat=&UniformAOrBFloat(
     A=0
     B=&RandomRangeFloat(Minimum=220, Maximum=280)
     Distribution Weight=0.66
     Evaluation Type=::NewEnumerator0
     Recalculate Random Each Loop=true
     Override Seed=false
    )
    Exponent=8
    Intensity:NiagaraFloat=&FNTimeOfDayScalar(Night Scale=5, Morning Scale=6, Day Scale=8, Evening Scale=6)
   )
   /SetVariables(
    Particles.ConeColor:LinearColor=&ColorFromCurve(ColorCurve=DataInterface() {propertyValues {
       AlphaCurve {keys [<at 0 0>,<at 0.2098 0.4 RCIM_Cubic>,<at 1 0 RCIM_Cubic>]}
       BlueCurve {keys [<at 0.2079 1 RCIM_Cubic>]}
       GreenCurve {keys [<at 0.2079 1 RCIM_Cubic>]}
       RedCurve {keys [<at 0.2079 1 RCIM_Cubic>]}
      }})
    Particles.ConeScale:Vector3f=&Multiply_Vector(
     A=<1 1 0.25>
     B=&VectorFromFloat(Value=&FloatFromCurve(FloatCurve=DataInterface() {propertyValues {Curve {keys [{arriveTangent=2.1576, interpMode=RCIM_Cubic, leaveTangent=2.1576, tangentMode=RCTM_User, time=0, value=0}, <at 1 1>]}}}))
    )
   ) {isSetParameters=true}
  ]
  rendererClasses [
   &NiagaraSpriteRendererProperties
   &NiagaraLightRendererProperties
   &NiagaraComponentRendererProperties
   &NiagaraMeshRendererProperties
  ]
  renderers [
   {rendererClass=&NiagaraSpriteRendererProperties}
   {rendererClass=&NiagaraSpriteRendererProperties, rendererIndex=1}
   {rendererClass=&NiagaraSpriteRendererProperties, rendererIndex=2}
   {rendererClass=&NiagaraSpriteRendererProperties, rendererIndex=3}
   {rendererClass=&NiagaraLightRendererProperties, rendererIndex=4}
   {rendererClass=&NiagaraComponentRendererProperties, rendererIndex=5}
   {rendererClass=&NiagaraMeshRendererProperties, rendererIndex=6}
   {rendererClass=&NiagaraSpriteRendererProperties, rendererIndex=7}
  ]
 }
 SpotLightMain {
  emitterUpdateScript [
   &EmitterState(Inactive Response=::NewEnumerator1, Life Cycle Mode=::NewEnumerator1, Loop Behavior=::NewEnumerator1, Loop Duration=14, UseLoopDelay=false)
   &SpawnBurst_Instantaneous(Spawn Count=2, Use Spawn Probability=false, UseLoopCountLimit=false)
  ]
  particleSpawnScript [
   &FN_SetSpawnGroup(Number of Spawn Groups=2)
   &InitializeParticle(
    Lifetime=&MaskFloatBySpawnGroup(Masked Float=11, Passthrough Float=2.73)
    Mesh Scale=&MaskVectorBySpawnGroup(Masked Vector=<0.8 0.8 1.2>, Passthrough Vector=<0.65 0.65 1.17>)
    Mesh Scale Mode=::NewEnumerator4
    Sprite Size Mode=::NewEnumerator3
    Uniform Sprite Size=&MaskFloatBySpawnGroup(Masked Float=250, Passthrough Float=200)
    UsePositionOffset=false
    Write Material Random=true
   )
   &CameraOffset()
   /SetVariables(
    Particles.FloorGlowFacingVector:Vector3f=<0 0 1>
    Particles.FloorGlowPosition:NiagaraPosition=&AddVectorToPosition(Position=&SimulationPosition(), Vector=<0 0 4>)
    Particles.FloorGlowSize:Vector2f=&Vector2DFromFloat(Value=250)
    Particles.LightAttenuationScale=275
    Particles.LightExponent=2
    Particles.LightPosition:NiagaraPosition=&AddVectorToPosition(Position=&SimulationPosition(), Vector=<0 0 100>)
    Particles.LightRadius=230
    Particles.SpotLightPosition:NiagaraPosition=&AddVectorToPosition(
     Position=&SimulationPosition()
     Vector=&MaskVectorBySpawnGroup(Masked Vector=<0 0 280>, Passthrough Vector=<-53.8 57.2 280>)
    )
   ) {isSetParameters=true}
   /SetVariables(Particles.LightVisibilityTag:NiagaraInt32=&MaskIntBySpawnGroup(Masked Int=0, Passthrough Int=1, Spawn Group Mask=0)) {isSetParameters=true}
  ]
  particleUpdateScript [
   &ParticleState()
   &Color(
    Color=&MaskLinearColorBySpawnGroup(
     Masked Linear Color=&ColorFromCurve(ColorCurve=DataInterface() {propertyValues {
        AlphaCurve {keys [<at 0 0.45 RCIM_Cubic>]}
        BlueCurve {keys [<at 0.0991 0.5938 RCIM_Cubic>,<at 0.3196 0.6206>,<at 0.5724 0.5>,<at 0.8766 0.4719>]}
        GreenCurve {keys [<at 0.0991 1.5513 RCIM_Cubic>,<at 0.3196 1.1957>,<at 0.5724 1.432>,<at 0.8766 1.1307>]}
        RedCurve {keys [<at 0.0991 2 RCIM_Cubic>,<at 0.3196 2>,<at 0.5724 2>,<at 0.8766 2>]}
       }})
     Passthrough Linear Color=&ColorFromCurve(ColorCurve=DataInterface() {propertyValues {
        AlphaCurve {keys [<at 0 0>,<at 0.0945 1 RCIM_Cubic>,<at 0.8978 1>,<at 1 0>]}
        BlueCurve {keys [<at 0.0991 0.4366 RCIM_Cubic>]}
        GreenCurve {keys [<at 0.0991 0.753 RCIM_Cubic>]}
        RedCurve {keys [<at 0.0991 1 RCIM_Cubic>]}
       }})
    )
    Scale Alpha=&FloatFromCurve(
     FloatCurve=DataInterface() {propertyValues {Curve {keys [
         <at 0 0 RCIM_Cubic>
         {interpMode=RCIM_Cubic, tangentMode=RCTM_User, time=0.03, value=1}
         <at 0.1364 0.8075 RCIM_Cubic>
         {arriveTangent=-0.3158, interpMode=RCIM_Cubic, leaveTangent=-0.3158, time=0.3, value=1}
         <at 0.4 0.7242 RCIM_Cubic>
         {arriveTangent=-0.0173, interpMode=RCIM_Cubic, leaveTangent=-0.0173, time=0.6, value=1}
         <at 0.7 0.719 RCIM_Cubic>
         {arriveTangent=1.0406, interpMode=RCIM_Cubic, leaveTangent=1.0406, time=0.8584, value=1}
         {interpMode=RCIM_Cubic, tangentMode=RCTM_User, time=0.97, value=1}
         <at 1 0>
        ]}}}
     Scale Curve=0.7
    )
   )
   /SetVariables(
    Particles.CoreSpotlightColor:LinearColor=&MultiplyLinearColor(
     Linear Color=<rgba 0.45 0.3224 0.1562 1>
     Scale Factor=&MakeFloat4FromVectorAndFloat("Vector (RGB)"=<1 1 1>, "Float (Alpha)"=1)
    )
    Particles.FlareColor=&MultiplyLinearColor(Linear Color=$Particles.Color, Scale Factor=&MakeVector4_FromFloat(Float=0.333))
    Particles.FloorGlowColor:LinearColor=&MultiplyLinearColor(Linear Color=$Particles.Color, Scale Factor=&MakeVector4_FromFloat(Float=0.666))
    Particles.LightColor=&MultiplyLinearColor(Linear Color=$Particles.Color, Scale Factor=&MakeVector4_FromFloat(Float=1.6))
    Particles.LightComponentIntensity:NiagaraFloat=&Multiply_Float(A=1, B=11111)
   ) {isSetParameters=true}
  ]
  rendererClasses [
   &NiagaraMeshRendererProperties
   &NiagaraLightRendererProperties
   &NiagaraSpriteRendererProperties
  ]
  renderers [
   {rendererClass=&NiagaraMeshRendererProperties, rendererIndex=0}
   {rendererClass=&NiagaraLightRendererProperties, rendererIndex=1}
   {rendererClass=&NiagaraSpriteRendererProperties, rendererIndex=2}
  ]
 }
}
summary {
 systemName=NS_Emote_Bestow
 emitters [
  {
   emitterName=PhotoFlashes
   rendererClasses [
    &NiagaraSpriteRendererProperties
    &NiagaraLightRendererProperties
    &NiagaraComponentRendererProperties
    &NiagaraMeshRendererProperties
   ]
  }
  {emitterName=ConfettiStreamers, rendererClasses [&NiagaraMeshRendererProperties]}
  {emitterName=ConfettiSquares, rendererClasses [&NiagaraSpriteRendererProperties]}
  {
   emitterName=SpotLightMain
   rendererClasses [
    &NiagaraMeshRendererProperties
    &NiagaraLightRendererProperties
    &NiagaraSpriteRendererProperties
   ]
  }
  {
   emitterName=GroundLightMesh
   rendererClasses [&NiagaraMeshRendererProperties, &NiagaraSpriteRendererProperties]
  }
 ]
 userVariables [
  User.bIsFrontEnd {type=NiagaraBool}
  User.bIsFrontEndPreview {type=NiagaraBool}
  User.ConfettiColorArray {
   type=NiagaraDataInterfaceArrayColor
   description=""
   defaultValue {
    struct=NiagaraExt_VariableValue_DataInterface
    value {
     dataInterface=&self:NiagaraDataInterfaceArrayColor_0
     dataInterfaceClass=&NiagaraDataInterfaceArrayColor
    }
   }
  }
 ]
}
